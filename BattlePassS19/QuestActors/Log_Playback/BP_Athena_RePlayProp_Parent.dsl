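; Parent Blueprint for Season 19 log-playback quest props that players discover or interact with to advance quest objectives.
; ConsolidatedQuestComponent events drive per-player visibility, with splitscreen handling; collision overlaps handle discovery, ambient FX and tag- or type-based completion on authority.
; ChangeStateForPlayer runs hide or swap visuals, cosmetics and feedback, choosing a scientist audio log by quest definition in the swap branch.
; A test mode bypasses quest requirements and resets visuals.
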(event OnCustomSetVisualstoPassed(ConsolidatedQuestComponent)
  (params (PlayerController :type Fort%20Player%20Controller%20Object%20Reference) (bFullyInitialized :type Boolean)))

(event OnCustomSetVisualstoFailed(ConsolidatedQuestComponent)
  (params (PlayerController :type Fort%20Player%20Controller%20Object%20Reference)))

(event Custom|ConversationToChangeState
  (params (PlayerPawn :type Actor%20Object%20Reference))
  ;[1 0.6502 0] Used at least once, but not much
  ;;[1 1 1] ConversationToChangeState
  (Networking|SwitchHasAuthority
    (:Authority
      (if (Variables|Design|GetbConversationToComplete)
        (CallFunction|CheckTagsandComplete PlayerPawn)))))

(event Custom|OnPlayerAction
  (params (PlayerController :type Fort%20Player%20Controller%20Object%20Reference) (ActionTag :type Gameplay%20Tag%20Structure) (PlayerTags :type Gameplay%20Tag%20Container%20Structure))
  (bind _btestvisualsnoquestrequired (Variables|Design|QuestSetup|GetbTestVisualsNoQuestRequired))
  ;[1 0.6502 0] Used at least once, but not much
  ;;[1 1 1] Player Action Tag to Interact
  (if (Variables|Default|GetbListenToTagToInteract)
    ;;;[1 1 1] Debugging - Display Tags
    (Development|PrintString (Utilities|String|Append "; Player Controller: " (Utilities|GetDisplayName PlayerController) "; Actor Tags: " (GameplayTags|GetDebugStringfromGameplayTagContainer PlayerTags)) _btestvisualsnoquestrequired _btestvisualsnoquestrequired "(R=0.000000,G=0.660000,B=1.000000,A=1.000000)" 3e+01)
    (Development|PrintString (Utilities|String|Append "[Received] Event Tag: " (GameplayTags|GetDebugStringfromGameplayTag ActionTag)) _btestvisualsnoquestrequired _btestvisualsnoquestrequired "(R=1.000000,G=0.084869,B=0.000000,A=1.000000)" 3e+01)
    (Networking|SwitchHasAuthority
      (:Authority
        (if (GameplayTags|HasAnyTags (GameplayTags|MakeGameplayTagContainerfromTag ActionTag) (Variables|Design|QuestSetup|InteractionSettings|GetDamage%20or%20Interact%20Tag))
          (CallFunction|CheckTagsandComplete PlayerController))))))

(event Custom|BindConversationEvents
  ;[1 1 1] ConversationEvents
  (EventDispatchers|BindEventtoOnConversationEnded (Variables|Components|GetNonPlayerConversationComponent) (delegate Custom|ConversationEnded)))

(event Custom|ConversationEnded
  ;[1 1 1] ConversationEvents
  (Variables|Default|SetbBlockLocalInteractOverride false))

(event EventBeginPlay
  (if (Utilities|IsValid (Variables|Components|GetNonPlayerConversationComponent) :!class /Script/Engine.KismetSystemLibrary)
    ;[1 1 1] Bind Conversation Events
    (CallFunction|BindConversationEvents))
  (if (Variables|Design|QuestSetup|GetbTestVisualsNoQuestRequired)
    (CallFunction|VisualTestingNoQuestRequiredInit)))

(event Custom|VisualTestingNoQuestRequired_Init
  (bind _p_ambient (Variables|Components|GetP_Ambient))
  ;[1 1 1] Visual Testing for when we don't want to bother granting ourself the quest
  (Components|Activation|SetAutoActivate _p_ambient false)
  (Components|Activation|Deactivate _p_ambient)
  (if (Variables|Design|QuestSetup|GetbRecheckingTriggersOnSuccess)
    (Utilities|Time|SetTimerbyEvent (delegate Custom|VisualTestingRecheckOverlapping) 0.5 true)))

(event Custom|VisualTestingRecheckOverlapping
  ;[1 1 1] Visual Testing for when we don't want to bother granting ourself the quest
  (CallFunction|UpdateOverlappingActors))

(event OnQuestNotReady(ConsolidatedQuestComponent)
  (params (PlayerController :type Fort%20Player%20Controller%20Object%20Reference))
  (bind _self self)
  (bind _sm_quest_postinteraction (Variables|Components|GetSM_Quest_PostInteraction))
  (bind _sm_quest_postinteraction_2 (Variables|Components|GetSM_Quest_PostInteraction))
  ;[0 0 0] QuestComponent Events
  ;;[1 1 1] OnQuestNotReady(QuestComponent)
  (if (Variables|Design|QuestSetup|GetbTestVisualsNoQuestRequired)
    _
    (elif (Utilities|Array|ContainsItem (Variables|Default|GetFortPlayerControllersWhoInteracted) PlayerController)
      ;;;[1 1 1] If so, don't hide the actors player interacted with
      (if (Quest|HasCompletedObjectivewithName (Quests|GetQuestManager PlayerController "Invalid") (Class|ConsolidatedQuestComponent|GetQuestItemDefinition (Variables|BuildingGameplayActorPropQuest|GetConsolidatedQuestComponent)) (Variables|/Script/FortniteGame.ConsolidatedQuestComponent|GetObjectiveBackendName (Variables|BuildingGameplayActorPropQuest|GetConsolidatedQuestComponent)) :!class /Script/FortniteGame.FortQuestManager)
        _
        (elif (Variables|Design|QuestSetup|GetbVisibleToAll)
          (if (Quest|HasCompletedObjectivewithName (Quests|GetQuestManager PlayerController "Invalid") (Class|ConsolidatedQuestComponent|GetQuestItemDefinition (Variables|BuildingGameplayActorPropQuest|GetConsolidatedQuestComponent)) (Variables|/Script/FortniteGame.ConsolidatedQuestComponent|GetObjectiveBackendName (Variables|BuildingGameplayActorPropQuest|GetConsolidatedQuestComponent)) :!class /Script/FortniteGame.FortQuestManager)
            ;;;[1 1 1] If player completed the quest, show the post-interaction version
            (CallFunction|PostInteractionVisuals PlayerController)
            (elif (PlayerController|IsinSplitscreen PlayerController)
              (PlayerController|SetCompHiddenControllerOnly PlayerController _sm_quest_postinteraction true true)
              (else
                (Development|SetHiddeninGame _sm_quest_postinteraction true true))))
          (elif (PlayerController|IsinSplitscreen PlayerController)
            (PlayerController|SetActorHiddenControllerOnly PlayerController _self true)
            (PlayerController|SetCompHiddenControllerOnly PlayerController _sm_quest_postinteraction_2 true false)
            (Rendering|SetCastShadow (Variables|Components|GetSM_Quest_PreInteraction) false)
            (Utilities|Map|Remove (Variables|BuildingGameplayActorPropQuest|GetFortPlayerControllerToCanInteract) PlayerController)
            (else
              (Rendering|SetActorHiddenInGame _self true)
              (Development|SetHiddeninGame _sm_quest_postinteraction_2 true false)))))
      (else
        _))))

(event Custom|ForceInteractStateChange
  (params (FPC :type Fort%20Player%20Controller%20Object%20Reference))
  ;[0 0 0] State changes & Cosmetics
  ;;[1 1 1] ForceInteractStateChange
  (CallFunction|LocalVisualsTest FPC)
  (CallFunction|ChangeStateforPlayer FPC)
  (Utilities|Array|AddUnique (Variables|Default|GetFortPlayerControllersWhoInteracted) FPC))

(event OnQuestReady(ConsolidatedQuestComponent)
  (params (PlayerController :type Fort%20Player%20Controller%20Object%20Reference))
  (bind _returnvalue (PlayerController|IsinSplitscreen PlayerController))
  (bind _self self)
  (bind _self_2 self)
  (bind _sm_quest_postinteraction (Variables|Components|GetSM_Quest_PostInteraction))
  ;[0 0 0] QuestComponent Events
  ;;[1 1 1] OnQuestReady(QuestComponent)
  (if (or (Quests|IsQuestObjectiveActive (Quests|GetQuestManager PlayerController "Invalid") (Variables|/Script/FortniteGame.ConsolidatedQuestComponent|GetObjectiveBackendName (Variables|BuildingGameplayActorPropQuest|GetConsolidatedQuestComponent))) (Variables|Design|QuestSetup|GetbTestVisualsNoQuestRequired))
    (if _returnvalue
      (PlayerController|SetActorHiddenControllerOnly PlayerController _self_2 false)
      (PlayerController|SetCompHiddenControllerOnly PlayerController _sm_quest_postinteraction true false)
      (else
        (Rendering|SetActorHiddenInGame _self_2 false)
        (Development|SetHiddeninGame _sm_quest_postinteraction true false)))
    (if (not (Variables|Design|GetbUseDiscoverCollisionToUnhide))
      (Utilities|Map|Add (Variables|BuildingGameplayActorPropQuest|GetFortPlayerControllerToCanInteract) PlayerController "true"))
    (elif (Variables|Design|QuestSetup|GetbVisibleToAll)
      (CallFunction|PostInteractionVisuals PlayerController)
      (elif _returnvalue
        (PlayerController|SetActorHiddenControllerOnly PlayerController _self true)
        (else
          (Rendering|SetActorHiddenInGame _self true))))))

(event Custom|ChangeStateForPlayer
  (params (FPC :type Fort%20Player%20Controller%20Object%20Reference))
  (bind _consolidatedquestcomponent (Variables|BuildingGameplayActorPropQuest|GetConsolidatedQuestComponent))
  (bind _returnvalue (Class|ConsolidatedQuestComponent|GetQuestItemDefinition _consolidatedquestcomponent))
  (bind _bshouldapplycosmeticoverrides (Variables|Design|QuestSetup|GameplayTags,EffectsandCosmetics|GetbShouldApplyCosmeticOverrides))
  (bind _sm_quest_preinteraction (Variables|Components|GetSM_Quest_PreInteraction))
  ;[0 0 0] State changes & Cosmetics
  ;;[1 1 1] Prop State Change on Interact
  (CallFunction|QuestLog "change state for player")
  (Variables|Debug|SetbOnVisualTestingCooldown true)
  (Utilities|Map|Remove (Variables|BuildingGameplayActorPropQuest|GetFortPlayerControllerToCanInteract) FPC)
  (CallFunction|QuestLog "remove controller interactions")
  ;;;[1 0 0] This needs to work on just the interacting player in splitscreen
  (Components|Activation|Deactivate (Variables|Components|GetP_Ambient))
  (Utilities|FlowControl|Sequence
    (:then_0
      (if (< (Variables|Design|CosmeticEvent|GetCosmeticTimeShift) 0.0)
        (Components|Activation|Activate (Variables|Components|GetP_Cosmetic) false)
        (CallFunction|SpawnCosmeticActor)
        (CallFunction|QuestLog "top delay")
        (Utilities|FlowControl|Delay (- (Math|Float|Absolute(Float) (Variables|Design|CosmeticEvent|GetCosmeticTimeShift)) (Variables|Design|PostInteraction|GetStateChangeDelay)))
        (else
          (CallFunction|QuestLog "bottom delay")))
      (Utilities|FlowControl|Delay (Variables|Design|PostInteraction|GetStateChangeDelay))
      (CallFunction|QuestLog "delay finished")
      (if (Variables|Default|GetbDamageToInteract)
        (Collision|SetCollisionEnabled (Variables|Components|GetSM_HiddenInteractionCollision) "NoCollision")
        (else
          _))
      ;;;[1 1 1] Hiding or Swapping?
      (if (Variables|Design|PostInteraction|GetbHideAfterInteract)
        (CallFunction|QuestLog "hide after interact")
        (AddComponent|NODE_AddStaticMeshComponent-0)
        (bind (_direction _scaledowntime) (Timeline|T_UnhideCurve PlayFromStart)
          (:Update
            (Transformation|SetWorldScale3D (Variables|Components|GetSM_Quest_PreInteraction) (Math|Vector|Lerp(Vector) (.scale (Variables|Design|PreInteraction|GetPreInteraction%20Mesh%20Transform)) "0.010000,0.010000,0.010000" _scaledowntime)))
          (:Finished
            (if (PlayerController|IsinSplitscreen FPC)
              (PlayerController|SetCompHiddenControllerOnly FPC _sm_quest_preinteraction true true)
              (else
                (Development|SetHiddeninGame _sm_quest_preinteraction true true)))
            (CallFunction|QuestLog "mesh now hidden in game")
            (Development|PrintString "Entered Post Interaction Logic" true true "(R=0.263882,G=1.000000,B=0.000000,A=1.000000)")
            ;;;[1 1 1] Post-Interaction Particles and Sound
            (CallFunction|QuestLog "Before playing reward feedback")
            (Class|AthenaQuestFunctionLibrary|AthenaPlayQuestRewardFeedback (Variables|Design|PostInteraction|SuccessFX|GetbPlayFXOnPlayer) (Variables|Design|PostInteraction|SuccessFX|GetSuccessSound) (Variables|Design|PostInteraction|SuccessFX|GetSuccessNiagara) 0 self (Utilities|Casting|CastToFortPlayerPawn (Pawn|GetControlledPawn FPC)) (Variables|Design|PostInteraction|SuccessFX|GetSuccessFXTransform))
            (Development|PrintString "PlayedFeedback" true true "(R=1.000000,G=0.000000,B=0.089822,A=1.000000)")
            (Networking|SwitchHasAuthority
              (:Authority
                ;;;;[1 1 1] Applying a Gameplay Effect to Player
                (CallFunction|ManagePlayerGEs FPC)
                (Development|PrintString "Has Cosmetic Override:")
                (Development|PrintString (Utilities|String|ToString(Boolean) _bshouldapplycosmeticoverrides))
                ;;;;[1 1 1] Applying Cosmetic Override to Player
                (if _bshouldapplycosmeticoverrides
                  (CallFunction|CosmeticOverridestoPlayer FPC)))
              (:Remote
                _))
            (Components|Activation|Activate (Variables|Components|GetPostInteractionAmbientAudio) false)
            (if (>= (Variables|Design|CosmeticEvent|GetCosmeticTimeShift) 0.0)
              (Utilities|FlowControl|Delay (Variables|Design|CosmeticEvent|GetCosmeticTimeShift))
              (Components|Activation|Activate (Variables|Components|GetP_Cosmetic) false)
              (CallFunction|SpawnCosmeticActor)
              (else
                _))
            (CallFunction|ResetForVisualTesting FPC)))
        (else
          (CallFunction|QuestLog "don't hide after interact")
          (CallFunction|PostInteractionVisuals FPC)
          (Utilities|FlowControl|DoOnce false
            (:Completed
              (if (== _returnvalue "/BattlePassS19/Items/QuestItems/Narrative/quest_s19_ReplayLog_01.quest_s19_ReplayLog_01")
                (Concatenator|Play (Variables|Default|GetScientist_AudioLog_01) (Variables|Components|GetAmbientAudio))
                (elif (== _returnvalue "/BattlePassS19/Items/QuestItems/Narrative/quest_s19_ReplayLog_02.quest_s19_ReplayLog_02")
                  (Concatenator|Play (Variables|Default|GetScientist_AudioLogs_02) (Variables|Components|GetAmbientAudio))
                  (elif (== _returnvalue "/BattlePassS19/Items/QuestItems/Narrative/quest_s19_ReplayLog_03.quest_s19_ReplayLog_03")
                    (Concatenator|Play (Variables|Default|GetScientist_AudioLogs_03) (Variables|Components|GetAmbientAudio))
                    (elif (== _returnvalue "/BattlePassS19/Items/QuestItems/Narrative/quest_s19_ReplayLog_04.quest_s19_ReplayLog_04")
                      (Concatenator|Play (Variables|Default|GetScientist_AudioLogs_04) (Variables|Components|GetAmbientAudio))
                      (elif (== _returnvalue "/BattlePassS19/Items/QuestItems/Narrative/quest_s19_ReplayLog_05.quest_s19_ReplayLog_05")
                        (Concatenator|Play (Variables|Default|GetScientist_AudioLogs_05) (Variables|Components|GetAmbientAudio))))))))))))
    (:then_1
      (Utilities|FlowControl|Delay 1e+01)
      (Environment|ApplyGlobalEnvironmentGameplayEffecttoActor (Game|Player|GetPlayerPawn FPC) "/BattlePassS19/QuestActors/Log_Playback/GE_RemoveAudioTag.GE_RemoveAudioTag_C"))))

(event Custom|LocalVisualsTest
  (params (FPC :type Fort%20Player%20Controller%20Object%20Reference))
  ;[0 0 0] State changes & Cosmetics
  ;;[1 1 1] Prop State Change on Interact
  ;;;[1 1 1] Visual Testing Only (replicated)
  (if (Utilities|Equal(Object) FPC (Fortnite|GetFirstLocalFortPlayerController true))
    (CallFunction|ChangeStateforPlayer FPC)))

(event OnUpdateQuest(ConsolidatedQuestComponent)
  (params (PlayerController :type Fort%20Player%20Controller%20Object%20Reference))
  (bind _self self)
  (bind _sm_quest_postinteraction (Variables|Components|GetSM_Quest_PostInteraction))
  ;[0 0 0] QuestComponent Events
  ;;[1 1 1] [QuestComponent] OnCustomUpdateQuest
  (if (or (Quests|IsQuestObjectiveActive (Quests|GetQuestManager PlayerController "Invalid") (Variables|/Script/FortniteGame.ConsolidatedQuestComponent|GetObjectiveBackendName (Variables|BuildingGameplayActorPropQuest|GetConsolidatedQuestComponent))) (Variables|Design|QuestSetup|GetbTestVisualsNoQuestRequired))
    (if (PlayerController|IsinSplitscreen PlayerController)
      (PlayerController|SetActorHiddenControllerOnly PlayerController _self false)
      (PlayerController|SetCompHiddenControllerOnly PlayerController _sm_quest_postinteraction true false)
      (else
        (Rendering|SetActorHiddenInGame _self false)
        (Development|SetHiddeninGame _sm_quest_postinteraction true false)))
    (if (not (Variables|Design|GetbUseDiscoverCollisionToUnhide))
      (Utilities|Map|Add (Variables|BuildingGameplayActorPropQuest|GetFortPlayerControllerToCanInteract) PlayerController "true"))))

(event Custom|Reset%20for%20Visual%20Testing
  (params (FPC :type Fort%20Player%20Controller%20Object%20Reference))
  (bind _sm_quest_preinteraction (Variables|Components|GetSM_Quest_PreInteraction))
  (bind _sm_quest_postinteraction (Variables|Components|GetSM_Quest_PostInteraction))
  (bind _preinteraction_mis_override (Variables|Design|PreInteraction|GetPreInteraction_MIS_Override))
  ;[0 0 0] State changes & Cosmetics
  ;;[1 1 1] Prop State Change on Interact
  ;;;[1 1 1] Reset to pre-interact state for visual testing mode
  (if (Variables|Design|QuestSetup|GetbTestVisualsNoQuestRequired)
    (Utilities|FlowControl|Delay (Variables|Design|QuestSetup|GetTestingCooldown))
    (Variables|Default|SetbBlockLocalInteractOverride false)
    (if (Variables|Design|PostInteraction|GetbHideAfterInteract)
      (Transformation|SetWorldScale3D (Variables|Components|GetSM_Quest_PreInteraction) (.scale (Variables|Design|PreInteraction|GetPreInteraction%20Mesh%20Transform)))
      (else
        _))
    (Components|Activation|Deactivate (Variables|Components|GetPostInteractionAmbientAudio))
    (Components|Activation|Activate (Variables|Components|GetAmbientAudio) false)
    (if (PlayerController|IsinSplitscreen FPC)
      (PlayerController|SetCompHiddenControllerOnly FPC _sm_quest_preinteraction false false)
      (PlayerController|SetCompHiddenControllerOnly FPC _sm_quest_postinteraction true false)
      (else
        (Development|SetHiddeninGame _sm_quest_preinteraction false false)
        (Development|SetHiddeninGame _sm_quest_postinteraction true false)))
    (Utilities|IsValid _preinteraction_mis_override
      (:Is%20Valid
        (Rendering|Material|SetMaterial (Variables|Components|GetSM_Quest_PreInteraction) 0 _preinteraction_mis_override))
      (:Is%20Not%20Valid
        _))
    (if (Variables|Design|QuestSetup|GameplayTags,EffectsandCosmetics|GetbShouldApplyCosmeticOverrides)
      (CallFunction|CosmeticOverridestoPlayer FPC true)
      (else
        _))
    (if (Variables|Default|GetbDamageToInteract)
      (Collision|SetCollisionEnabled (Variables|Components|GetSM_HiddenInteractionCollision) "QueryOnly"))
    (Variables|Debug|SetbOnVisualTestingCooldown false)))

(event OnQuestComponentSuccessEvent(ConsolidatedQuestComponent)
  (params (PlayerController :type Fort%20Player%20Controller%20Object%20Reference))
  (bind _consolidatedquestcomponent (Variables|BuildingGameplayActorPropQuest|GetConsolidatedQuestComponent))
  (bind _returnvalue (Class|ConsolidatedQuestComponent|GetQuestItemDefinition _consolidatedquestcomponent))
  (bind _returnvalue_2 (Quests|GetQuestManager PlayerController "Invalid"))
  (bind _returnvalue_3 (Quests|GetQuestwithDefinition _returnvalue_2 _returnvalue))
  ;[0 0 0] QuestComponent Events
  ;;[1 1 1] OnQuestComponentSuccessEvent
  (CallFunction|UpdateOverlappingActors)
  (CallFunction|UpdateVisuals (Pawn|GetControlledPawn PlayerController))
  (Networking|SwitchHasAuthority
    (:Authority
      (Utilities|IsValid _returnvalue_3
        (:Is%20Valid
          (if (Quest|HasCompletedAllObjectives _returnvalue_3 false)
            (if (Variables|Design|QuestSetup|GameplayTags,EffectsandCosmetics|GetbShouldApplyCosmeticOverrides)
              (if (Variables|Design|QuestSetup|GameplayTags,EffectsandCosmetics|GetbResetOverridesOnCompletion)
                (CallFunction|CosmeticOverridestoPlayer PlayerController true)
                (Development|PrintString "REMOVED" (Variables|Design|QuestSetup|GetbTestVisualsNoQuestRequired) (Variables|Design|QuestSetup|GetbTestVisualsNoQuestRequired) "(R=0.000000,G=1.000000,B=0.073006,A=1.000000)")))
            (else
              (Development|PrintString (Utilities|GetDisplayName _returnvalue)))))))))

(event OnCustomUpdateObjective(ConsolidatedQuestComponent)
  (params (PlayerController :type Fort%20Player%20Controller%20Object%20Reference) (AchievedCount :type Integer) (RequiredCount :type Integer))
  ;[0 0 0] QuestComponent Events
  ;;[1 1 1] OnCustomUpdateObjective(QuestComponent)
  ;;;[1 1 1] In Testing Mode The Cosmetic Event Was called already
  (if (Variables|Design|QuestSetup|GetbTestVisualsNoQuestRequired)
    _
    (elif (Variables|Design|GetbCompleteFromDiscovery)
      _
      (else
        ;;;[1 1 1] do we depend on the server rep from the quest manager for having the quest completed or assume they did?
        ;;;
        ;;; connect this backup if we want the former
        (CallFunction|ChangeStateforPlayer PlayerController)
        (Utilities|Array|AddUnique (Variables|Default|GetFortPlayerControllersWhoInteracted) PlayerController)))))

(event EventFxTestOverride
  (params (FortPC :type Fort%20Player%20Controller%20Object%20Reference))
  ;[0 0 0] State changes & Cosmetics
  ;;[1 1 1] Discovery Collision Overlap
  ;;;[1 1 1] Server needs to track the interact or send stat event for complex custom for "discovery'
  (CallFunction|ChangeStateforPlayer FortPC))

(event EventHideMesh
  (params (bIsInSplitScreen :type Boolean))
  (bind _sm_quest_preinteraction (Variables|Components|GetSM_Quest_PreInteraction))
  ;[0 0 0] State changes & Cosmetics
  ;;[1 1 1] Discovery Collision Overlap
  (if bIsInSplitScreen
    (PlayerController|SetCompHiddenControllerOnly (Variables|BuildingGameplayActorPropQuest|GetFortPlayerController_TempRef) _sm_quest_preinteraction false true)
    (bind (_direction _scaledowntime) (Timeline|Timeline_1 PlayFromStart)
      (:Update
        (Transformation|SetWorldScale3D (Variables|Components|GetSM_Quest_PreInteraction) (Math|Vector|Lerp(Vector) "0.010000,0.010000,0.010000" (.scale (Variables|Design|PreInteraction|GetPreInteraction%20Mesh%20Transform)) _scaledowntime))))
    (else
      (Development|SetHiddeninGame _sm_quest_preinteraction false true)
      (Timeline|Timeline_1 PlayFromStart))))

(event Custom|UpdateOverlappingActors(Collision_DiscoverRadius)
  (params (Other%20Actor :type Actor%20Object%20Reference))
  ;[0 0 0] State changes & Cosmetics
  ;;[1 1 1] Discovery Collision Overlap
  (bind _data (Utilities|FlowControl|MergeExecutionwithData Other%20Actor (stale OnComponentBeginOverlap(Collision_DiscoverRadius) OtherActor))
    (:Then
      (CallFunction|HandleComponentOverlapDisoveryRadius _data (Utilities|Struct|MakeFortOverlapData (Variables|Design|QuestSetup|GameplayTags,EffectsandCosmetics|GetRequired%20Player%20Tags) (Variables|Design|QuestSetup|GameplayTags,EffectsandCosmetics|GetProhibited%20Player%20Tags) (Variables|Design|QuestSetup|GetbReHideIfConditionsChange) (Variables|Design|GetbCompleteFromDiscovery) (Variables|Design|QuestSetup|GetbTestVisualsNoQuestRequired) (Variables|Debug|GetbOnVisualTestingCooldown))))))

(event OnComponentBeginOverlap(Collision_DiscoverRadius)
  (params (OverlappedComponent :type Primitive%20Component%20Object%20Reference) (OtherActor :type Actor%20Object%20Reference) (OtherComp :type Primitive%20Component%20Object%20Reference) (OtherBodyIndex :type Integer) (bFromSweep :type Boolean) (SweepResult :type Hit%20Result%20Structure%20%28by%20ref%29))
  ;[0 0 0] State changes & Cosmetics
  ;;[1 1 1] Discovery Collision Overlap
  (bind _data (Utilities|FlowControl|MergeExecutionwithData (stale Custom|UpdateOverlappingActors(Collision_DiscoverRadius) Other%20Actor) OtherActor)
    (:Then
      (CallFunction|HandleComponentOverlapDisoveryRadius _data (Utilities|Struct|MakeFortOverlapData (Variables|Design|QuestSetup|GameplayTags,EffectsandCosmetics|GetRequired%20Player%20Tags) (Variables|Design|QuestSetup|GameplayTags,EffectsandCosmetics|GetProhibited%20Player%20Tags) (Variables|Design|QuestSetup|GetbReHideIfConditionsChange) (Variables|Design|GetbCompleteFromDiscovery) (Variables|Design|QuestSetup|GetbTestVisualsNoQuestRequired) (Variables|Debug|GetbOnVisualTestingCooldown))))))

(event EventCompleteFromDiscovery
  ;[0 0 0] State changes & Cosmetics
  ;;[1 1 1] Discovery Collision Overlap
  ;;;[1 1 1] Complete the Quest From Discovery
  (Class|AthenaQuestFunctionLibrary|AthenaPlayQuestRewardFeedback (Variables|Design|PostInteraction|SuccessFX|GetbPlayFXOnPlayer) (Variables|Design|PostInteraction|SuccessFX|GetSuccessSound) (Variables|Design|PostInteraction|SuccessFX|GetSuccessNiagara) 0 self (Game|Player|GetPlayerPawn (Variables|BuildingGameplayActorPropQuest|GetFortPlayerController_TempRef)) (Variables|Design|PostInteraction|SuccessFX|GetSuccessFXTransform)))

(event EventActivateComponents
  ;[0 0 0] State changes & Cosmetics
  ;;[1 1 1] Discovery Collision Overlap
  ;;;[1 1 1] Activate Components
  (Components|Activation|Activate (Variables|Components|GetUnHideAudio) false)
  (Components|Activation|Activate (Variables|Components|GetP_UnHide) true)
  (Components|Activation|Activate (Variables|Components|GetAmbientAudio) false)
  (Components|Activation|Activate (Variables|Components|GetN_Ambient) true))

(event OnComponentEndOverlap(Collision_DiscoverRadius)
  (params (OverlappedComponent :type Primitive%20Component%20Object%20Reference) (OtherActor :type Actor%20Object%20Reference) (OtherComp :type Primitive%20Component%20Object%20Reference) (OtherBodyIndex :type Integer))
  ;[0 0 0] State changes & Cosmetics
  ;;[1 1 1] Discovery Collision Overlap
  ;;;[1 0.1414 0] This is gnarly, don't like this - for future optimization (Woj)
  (bind _asfort_player_pawn_athena (Utilities|Casting|CastToFortPlayerPawnAthena OtherActor))
  (if (Pawn|IsLocallyControlled _asfort_player_pawn_athena)
    (CallFunction|BPReHideActor (Fortnite|GetFortPlayerControllerfromActor _asfort_player_pawn_athena)))
  (Utilities|Array|RemoveItem (Variables|Default|GetbOverlapWatchedActors) _asfort_player_pawn_athena))

(event Custom|BP_ReHideActor
  (params (Fort%20PC :type Fort%20Player%20Controller%20Object%20Reference))
  (bind (_value _returnvalue) (Utilities|Map|Find (Variables|BuildingGameplayActorPropQuest|GetFortPlayerControllerToCanInteract) (Variables|BuildingGameplayActorPropQuest|GetFortPlayerController_TempRef)))
  (bind _sm_quest_preinteraction (Variables|Components|GetSM_Quest_PreInteraction))
  ;[0 0 0] State changes & Cosmetics
  ;;[1 1 1] Discovery Collision Overlap
  (bind _data (Utilities|FlowControl|MergeExecutionwithData Fort%20PC (stale EventReHideActor FortPC))
    (:Then
      (Variables|BuildingGameplayActorPropQuest|SetFortPlayerController_TempRef _data)
      (if (and _value _returnvalue)
        (bind (_direction _scaledowntime) (Timeline|Timeline_0 ReverseFromEnd)
          (:Update
            (Transformation|SetWorldScale3D (Variables|Components|GetSM_Quest_PreInteraction) (Math|Vector|Lerp(Vector) "0.010000,0.010000,0.010000" (.scale (Variables|Design|PreInteraction|GetPreInteraction%20Mesh%20Transform)) _scaledowntime)))
          (:Finished
            (if (PlayerController|IsinSplitscreen (Variables|BuildingGameplayActorPropQuest|GetFortPlayerController_TempRef))
              (PlayerController|SetCompHiddenControllerOnly (Variables|BuildingGameplayActorPropQuest|GetFortPlayerController_TempRef) _sm_quest_preinteraction true true)
              (else
                (Development|SetHiddeninGame _sm_quest_preinteraction true true)))
            ;;;[1 1 1] Add can interact on client
            (Utilities|Map|Add (Variables|BuildingGameplayActorPropQuest|GetFortPlayerControllerToCanInteract) (Variables|BuildingGameplayActorPropQuest|GetFortPlayerController_TempRef))
            (Components|Activation|Deactivate (Variables|Components|GetAmbientAudio))
            (Components|Activation|Deactivate (Variables|Components|GetN_Ambient))))))))

(event EventReHideActor
  (params (FortPC :type Fort%20Player%20Controller%20Object%20Reference))
  (bind (_value _returnvalue) (Utilities|Map|Find (Variables|BuildingGameplayActorPropQuest|GetFortPlayerControllerToCanInteract) (Variables|BuildingGameplayActorPropQuest|GetFortPlayerController_TempRef)))
  (bind _sm_quest_preinteraction (Variables|Components|GetSM_Quest_PreInteraction))
  ;[0 0 0] State changes & Cosmetics
  ;;[1 1 1] Discovery Collision Overlap
  (bind _data (Utilities|FlowControl|MergeExecutionwithData (stale Custom|BP_ReHideActor Fort%20PC) FortPC)
    (:Then
      (Variables|BuildingGameplayActorPropQuest|SetFortPlayerController_TempRef _data)
      (if (and _value _returnvalue)
        (Timeline|Timeline_0 ReverseFromEnd)))))

(event OnComponentBeginOverlap(SM_HiddenInteractionCollision)
  (params (OverlappedComponent :type Primitive%20Component%20Object%20Reference) (OtherActor :type Actor%20Object%20Reference) (OtherComp :type Primitive%20Component%20Object%20Reference) (OtherBodyIndex :type Integer) (bFromSweep :type Boolean) (SweepResult :type Hit%20Result%20Structure%20%28by%20ref%29))
  ;[0 0 0] State changes & Cosmetics
  ;;[1 1 1] Overlap to Change State - Listen For Tag, Vehicles etc.
  (bind _data (Utilities|FlowControl|MergeExecutionwithData OtherActor (stale Custom|UpdateOverlappingActors(SM_HiddenInteractionCollision) OtherActor))
    (:Then
      (Networking|SwitchHasAuthority
        (:Authority
          (if (Variables|Default|GetbOverlapToInteract)
            (if (Variables|Default|GetbListenToTagToInteract)
              (bind _asfort_player_pawn (Utilities|Casting|CastToFortPlayerPawn _data))
              (if (Quests|IsQuestObjectiveActive (Quests|GetQuestManager (Fortnite|GetFortPlayerControllerfromActor _asfort_player_pawn) "Invalid") (Variables|/Script/FortniteGame.ConsolidatedQuestComponent|GetObjectiveBackendName (Variables|BuildingGameplayActorPropQuest|GetConsolidatedQuestComponent)))
                _
                (elif (Variables|Design|QuestSetup|GetbTestVisualsNoQuestRequired)
                  _))
              (GameplayEffects|ApplyGameplayEffectToTarget :self (Ability|GetAbilitySystemComponent self) :GameplayEffectClass (Variables|Default|GetTagListenerEffect) :Target (Ability|GetAbilitySystemComponent _asfort_player_pawn) :Level 0.0)
              (CallFunction|PlayerinInteractRange _asfort_player_pawn)
              (else
                (switch Utilities|FlowControl|Switch|SwitchonE_Athena_GenericQuestDamageTypes (Variables|Design|QuestSetup|InteractionSettings|GetInteractionType)
                  (:NewEnumerator9
                    ;;;[1 1 1] Overlap  to Interact
                    (bind _asfort_vehicle_interface (Utilities|Casting|CastToFortVehicleInterface _data))
                    (for _array_element (Vehicle|GetAllPassengers _asfort_vehicle_interface)
                      (CallFunction|CheckTagsandComplete _array_element)))
                  (:NewEnumerator18
                    (bind _asphysics_ball_master (Utilities|Casting|CastToPhysicsBall_Master _data))
                    (CallFunction|CheckTagsandComplete (Actor|GetOwner _asphysics_ball_master)))
                  (:NewEnumerator20
                    (bind _asfort_player_pawn_2 (Utilities|Casting|CastToFortPlayerPawn _data))
                    (CallFunction|CheckTagsandComplete _asfort_player_pawn_2))
                  (:NewEnumerator21
                    (bind _asfort_vehicle_interface_2 (Utilities|Casting|CastToFortVehicleInterface _data)
                      (:then
                        (for _array_element_2 (Vehicle|GetAllPassengers _asfort_vehicle_interface_2)
                          (CallFunction|CheckTagsandComplete _array_element_2)))
                      (:CastFailed
                        (bind _asfort_player_pawn_3 (Utilities|Casting|CastToFortPlayerPawn _data))
                        (CallFunction|CheckTagsandComplete _asfort_player_pawn_3)))))))))))))

(event OnComponentBeginOverlap(Collision_AmbientAudioAndFXActivation)
  (params (OverlappedComponent :type Primitive%20Component%20Object%20Reference) (OtherActor :type Actor%20Object%20Reference) (OtherComp :type Primitive%20Component%20Object%20Reference) (OtherBodyIndex :type Integer) (bFromSweep :type Boolean) (SweepResult :type Hit%20Result%20Structure%20%28by%20ref%29))
  (bind _consolidatedquestcomponent (Variables|BuildingGameplayActorPropQuest|GetConsolidatedQuestComponent))
  ;[0 0 0] State changes & Cosmetics
  ;;[1 1 1] Activate Ambient Audio & FX if in test state or obj active
  (bind _data (Utilities|FlowControl|MergeExecutionwithData OtherActor (stale Custom|UpdateVisuals Other%20Actor))
    (:Then
      (bind _asfort_player_pawn_athena (Utilities|Casting|CastToFortPlayerPawnAthena _data))
      (Networking|SwitchHasAuthority
        (:Remote
          (bind (_playercontrollerout _completedquest) (Quest|HasPlayerCompletedQuest (Fortnite|GetFortPlayerControllerfromActor _asfort_player_pawn_athena) (Class|ConsolidatedQuestComponent|GetQuestItemDefinition _consolidatedquestcomponent)))
          (if (Math|Boolean|ANDBoolean (not (Variables|Design|GetbUseDiscoverCollisionToUnhide)) (or (Math|Boolean|ANDBoolean (not _completedquest) (Quests|IsQuestObjectiveActive (Quests|GetQuestManager (Fortnite|GetFortPlayerControllerfromActor _asfort_player_pawn_athena) "Invalid") (Variables|/Script/FortniteGame.ConsolidatedQuestComponent|GetObjectiveBackendName _consolidatedquestcomponent)) (not (Variables|Design|QuestSetup|GetbTestVisualsNoQuestRequired))) (and (Variables|Design|QuestSetup|GetbTestVisualsNoQuestRequired) (not (Variables|Debug|GetbOnVisualTestingCooldown)))) (Interaction|BlueprintCanInteract _asfort_player_pawn_athena "AllInteraction"))
            (Components|Activation|Activate (Variables|Components|GetAmbientAudio) false)
            (Components|Activation|Activate (Variables|Components|GetP_Ambient) true)))))))

(event Custom|UpdateOverlappingActors(SM_HiddenInteractionCollision)
  (params (OtherActor :type Actor%20Object%20Reference))
  ;[0 0 0] State changes & Cosmetics
  ;;[1 1 1] Overlap to Change State - Listen For Tag, Vehicles etc.
  (bind _data (Utilities|FlowControl|MergeExecutionwithData (stale OnComponentBeginOverlap(SM_HiddenInteractionCollision) OtherActor) OtherActor)
    (:Then
      (Networking|SwitchHasAuthority
        (:Authority
          (if (Variables|Default|GetbOverlapToInteract)
            (if (Variables|Default|GetbListenToTagToInteract)
              (bind _asfort_player_pawn (Utilities|Casting|CastToFortPlayerPawn _data))
              (if (Quests|IsQuestObjectiveActive (Quests|GetQuestManager (Fortnite|GetFortPlayerControllerfromActor _asfort_player_pawn) "Invalid") (Variables|/Script/FortniteGame.ConsolidatedQuestComponent|GetObjectiveBackendName (Variables|BuildingGameplayActorPropQuest|GetConsolidatedQuestComponent)))
                _
                (elif (Variables|Design|QuestSetup|GetbTestVisualsNoQuestRequired)
                  _))
              (GameplayEffects|ApplyGameplayEffectToTarget :self (Ability|GetAbilitySystemComponent self) :GameplayEffectClass (Variables|Default|GetTagListenerEffect) :Target (Ability|GetAbilitySystemComponent _asfort_player_pawn) :Level 0.0)
              (CallFunction|PlayerinInteractRange _asfort_player_pawn)
              (else
                (switch Utilities|FlowControl|Switch|SwitchonE_Athena_GenericQuestDamageTypes (Variables|Design|QuestSetup|InteractionSettings|GetInteractionType)
                  (:NewEnumerator9
                    ;;;[1 1 1] Overlap  to Interact
                    (bind _asfort_vehicle_interface (Utilities|Casting|CastToFortVehicleInterface _data))
                    (for _array_element (Vehicle|GetAllPassengers _asfort_vehicle_interface)
                      (CallFunction|CheckTagsandComplete _array_element)))
                  (:NewEnumerator18
                    (bind _asphysics_ball_master (Utilities|Casting|CastToPhysicsBall_Master _data))
                    (CallFunction|CheckTagsandComplete (Actor|GetOwner _asphysics_ball_master)))
                  (:NewEnumerator20
                    (bind _asfort_player_pawn_2 (Utilities|Casting|CastToFortPlayerPawn _data))
                    (CallFunction|CheckTagsandComplete _asfort_player_pawn_2))
                  (:NewEnumerator21
                    (bind _asfort_vehicle_interface_2 (Utilities|Casting|CastToFortVehicleInterface _data)
                      (:then
                        (for _array_element_2 (Vehicle|GetAllPassengers _asfort_vehicle_interface_2)
                          (CallFunction|CheckTagsandComplete _array_element_2)))
                      (:CastFailed
                        (bind _asfort_player_pawn_3 (Utilities|Casting|CastToFortPlayerPawn _data))
                        (CallFunction|CheckTagsandComplete _asfort_player_pawn_3)))))))))))))

(event Custom|UpdateVisuals
  (params (Other%20Actor :type Actor%20Object%20Reference))
  (bind _returnvalue (Fortnite|GetFortPlayerControllerfromActor Other%20Actor))
  (bind _consolidatedquestcomponent (Variables|BuildingGameplayActorPropQuest|GetConsolidatedQuestComponent))
  ;[0 0 0] State changes & Cosmetics
  ;;[1 1 1] Activate Ambient Audio & FX if in test state or obj active
  (Utilities|IsValid _returnvalue
    (:Is%20Valid
      (if (Pawn|IsLocalPlayerController _returnvalue)
        (Utilities|FlowControl|Delay 0.5)
        (bind _data (Utilities|FlowControl|MergeExecutionwithData (stale OnComponentBeginOverlap(Collision_AmbientAudioAndFXActivation) OtherActor) Other%20Actor)
          (:Then
            (bind _asfort_player_pawn_athena (Utilities|Casting|CastToFortPlayerPawnAthena _data))
            (Networking|SwitchHasAuthority
              (:Remote
                (bind (_playercontrollerout _completedquest) (Quest|HasPlayerCompletedQuest (Fortnite|GetFortPlayerControllerfromActor _asfort_player_pawn_athena) (Class|ConsolidatedQuestComponent|GetQuestItemDefinition _consolidatedquestcomponent)))
                (if (Math|Boolean|ANDBoolean (not (Variables|Design|GetbUseDiscoverCollisionToUnhide)) (or (Math|Boolean|ANDBoolean (not _completedquest) (Quests|IsQuestObjectiveActive (Quests|GetQuestManager (Fortnite|GetFortPlayerControllerfromActor _asfort_player_pawn_athena) "Invalid") (Variables|/Script/FortniteGame.ConsolidatedQuestComponent|GetObjectiveBackendName _consolidatedquestcomponent)) (not (Variables|Design|QuestSetup|GetbTestVisualsNoQuestRequired))) (and (Variables|Design|QuestSetup|GetbTestVisualsNoQuestRequired) (not (Variables|Debug|GetbOnVisualTestingCooldown)))) (Interaction|BlueprintCanInteract _asfort_player_pawn_athena "AllInteraction"))
                  (Components|Activation|Activate (Variables|Components|GetAmbientAudio) false)
                  (Components|Activation|Activate (Variables|Components|GetP_Ambient) true))))))))))

(event OnComponentEndOverlap(Collision_AmbientAudioAndFXActivation)
  (params (OverlappedComponent :type Primitive%20Component%20Object%20Reference) (OtherActor :type Actor%20Object%20Reference) (OtherComp :type Primitive%20Component%20Object%20Reference) (OtherBodyIndex :type Integer))
  ;[0 0 0] State changes & Cosmetics
  ;;[1 1 1] Disabling Ambient Sound and VFX on moving away
  (Utilities|Casting|CastToFortPlayerPawnAthena OtherActor)
  (Networking|SwitchHasAuthority
    (:Remote
      (Components|Activation|Deactivate (Variables|Components|GetAmbientAudio))
      (Components|Activation|Deactivate (Variables|Components|GetP_Ambient)))))

(event Custom|PlayerInInteractRange
  (params (Actor :type Actor%20Object%20Reference)))

(event OnComponentEndOverlap(SM_HiddenInteractionCollision)
  (params (OverlappedComponent :type Primitive%20Component%20Object%20Reference) (OtherActor :type Actor%20Object%20Reference) (OtherComp :type Primitive%20Component%20Object%20Reference) (OtherBodyIndex :type Integer))
  ;[0 0 0] State changes & Cosmetics
  ;;[1 1 1] Overlap to Change State - Listen For Tag, Vehicles etc.
  (Networking|SwitchHasAuthority
    (:Authority
      (if (Variables|Default|GetbOverlapToInteract)
        (if (Variables|Default|GetbListenToTagToInteract)
          (bind _asfort_player_pawn (Utilities|Casting|CastToFortPlayerPawn OtherActor))
          (GameplayEffects|RemoveActiveGameplayEffectbySourceEffect (Ability|GetAbilitySystemComponent _asfort_player_pawn) (Variables|Default|GetTagListenerEffect) (Ability|GetAbilitySystemComponent self) -1)
          (CallFunction|PlayerOutOfInteractRange _asfort_player_pawn))))))

(event Custom|PlayerOutOfInteractRange
  (params (Actor :type Actor%20Object%20Reference)))

(event EventOnDamageServer
  (params (Damage :type Float%20%28single-precision%29) (DamageTags :type Gameplay%20Tag%20Container%20Structure%20%28by%20ref%29) (Momentum :type Vector :default "0, 0, 0") (HitInfo :type Hit%20Result%20Structure%20%28by%20ref%29) (InstigatedBy :type Controller%20Object%20Reference) (DamageCauser :type Actor%20Object%20Reference) (EffectContext :type Gameplay%20Effect%20Context%20Handle%20Structure))
  ;[0 0 0] State changes & Cosmetics
  ;;[1 1 1] Damage to Change State
  (if (Variables|Default|GetbDamageToInteract)
    (if (GameplayTags|HasAnyTags DamageTags (Variables|Design|QuestSetup|InteractionSettings|GetDamage%20or%20Interact%20Tag))
      (bind _asfort_player_controller (Utilities|Casting|CastToFortPlayerController (Utilities|SelectObject InstigatedBy (Ability|EffectContext|GetEffectCauser EffectContext) (Utilities|IsValid InstigatedBy :!class /Script/Engine.KismetSystemLibrary))))
      (CallFunction|CheckTagsandComplete _asfort_player_controller))))

(fn BlueprintCanInteract
  (params (InteractingPawn :type Fort%20Pawn%20Object%20Reference) (InteractionBeingAttempted :type EInteractionBeingAttempted%20Enum :default "FirstInteraction") (InteractionType :type TInteractionType%20Enum :default "IT_NoInteraction"))
  (bind _consolidatedquestcomponent (Variables|BuildingGameplayActorPropQuest|GetConsolidatedQuestComponent))
  (bind _returnvalue (Class|ConsolidatedQuestComponent|GetQuestItemDefinition _consolidatedquestcomponent))
  (bind _objectivebackendname (Variables|/Script/FortniteGame.ConsolidatedQuestComponent|GetObjectiveBackendName _consolidatedquestcomponent))
  (bind _asfort_player_controller (Utilities|Casting|CastToFortPlayerController (Pawn|GetController InteractingPawn))
    (:then
      (bind _bdidactorhavetags (Class|AthenaQuestFunctionLibrary|AthenaCheckActorHasTags (Game|Player|GetPlayerPawn _asfort_player_controller) "(GameplayTags=((TagName=\"Athena.Quests.Interact.ReplayLog.PlayingLog\")))"))
      (if _bdidactorhavetags
        (return false)
        (else
          (LocalVariables|SetLocal_FortPlayerControllerAthena _asfort_player_controller)
          (if (Quests|IsQuestObjectiveActive (Quests|GetQuestManager _asfort_player_controller "Invalid") _objectivebackendname)
            _
            (elif (Quest|HasCompletedQuest (Quests|GetQuestManager _asfort_player_controller "Invalid") _returnvalue :!class /Script/FortniteGame.FortQuestManager)
              _))
          (Networking|SwitchHasAuthority
            (:Authority
              (return (Math|Boolean|ANDBoolean (Math|Boolean|ORBoolean (Quests|IsQuestObjectiveActive (Quests|GetQuestManager _asfort_player_controller "Invalid") _objectivebackendname) (Variables|Design|QuestSetup|GetbTestVisualsNoQuestRequired) (Quest|HasCompletedQuest (Quests|GetQuestManager _asfort_player_controller "Invalid") _returnvalue :!class /Script/FortniteGame.FortQuestManager)) (and (not (GameplayTags|HasAnyTags (GameplayTags|GetOwnedGameplayTags InteractingPawn :!class /Script/GameplayTags.GameplayTagAssetInterface) (Variables|Design|QuestSetup|GameplayTags,EffectsandCosmetics|GetProhibited%20Player%20Tags) true)) (GameplayTags|HasAllTags (GameplayTags|GetOwnedGameplayTags InteractingPawn :!class /Script/GameplayTags.GameplayTagAssetInterface) (Variables|Design|QuestSetup|GameplayTags,EffectsandCosmetics|GetRequired%20Player%20Tags) true)) (or (not (Variables|Default|GetbDepositToInteract)) (>= (Inventory|GetItemQuantityOnPlayer _asfort_player_controller (Variables|Default|GetResourceItemDefinition)) (Variables|Design|QuestSetup|InteractionSettings|GetDepositAmount))))))
            (:Remote
              (if (Variables|Default|GetbBlockLocalInteractOverride)
                (return false)
                (else
                  (return true))))))))
    (:CastFailed
      (return false))))

(fn BlueprintGetFailedInteractionString
  (params (InteractingPawn :type Fort%20Pawn%20Object%20Reference) (InteractionBeingAttempted :type EInteractionBeingAttempted%20Enum :default "FirstInteraction"))
  (bind _returnvalue (Fortnite|GetFortPlayerControllerfromActor InteractingPawn))
  (bind _amountneeded (LocalVariables|GetAmountNeeded))
  (if (Variables|Default|GetbDepositToInteract)
    (if (or (Quests|IsQuestObjectiveActive (Quests|GetQuestManager _returnvalue "Invalid") (Variables|/Script/FortniteGame.ConsolidatedQuestComponent|GetObjectiveBackendName (Variables|BuildingGameplayActorPropQuest|GetConsolidatedQuestComponent))) (and (Variables|Design|QuestSetup|GetbTestVisualsNoQuestRequired) (not (Variables|Debug|GetbOnVisualTestingCooldown))))
      (bind _output_get (LocalVariables|SetAmountNeeded (- (Variables|Design|QuestSetup|InteractionSettings|GetDepositAmount) (Inventory|GetItemQuantityOnPlayer _returnvalue (Variables|Default|GetResourceItemDefinition)))))
      (if (> _output_get 0)
        (switch Utilities|FlowControl|Switch|SwitchonE_Athena_GenericQuestDamageTypes (Variables|Design|QuestSetup|InteractionSettings|GetInteractionType)
          (:NewEnumerator13
            (return (Utilities|Text|FormatText "NSLOCTEXT(\"[F0DB40674147727E909EBBB40FEF4ECB]\", \"F78725A14F616F7A797DDC928A48032A\", \"Need {amount} more bars\")" _amountneeded)))
          (:NewEnumerator14
            (return (Utilities|Text|FormatText "NSLOCTEXT(\"[F0DB40674147727E909EBBB40FEF4ECB]\", \"3846AC2A418BB71AAA518FA8276F993F\", \"Need {amount} more wood\")" _amountneeded)))
          (:NewEnumerator15
            (return (Utilities|Text|FormatText "NSLOCTEXT(\"[F0DB40674147727E909EBBB40FEF4ECB]\", \"A6628577457EDD92A5028F99D2706C57\", \"Need {amount} more stone\")" _amountneeded)))
          (:NewEnumerator16
            (return (Utilities|Text|FormatText "NSLOCTEXT(\"[F0DB40674147727E909EBBB40FEF4ECB]\", \"453287044CF5446C686B1584A28EB013\", \"Need {amount} more metal\")" _amountneeded))))
        (else
          (return "")))
      (else
        _))
    (else
      _))
  (return ""))

(fn BlueprintGetInteractionString
  (params (InteractingPawn :type Fort%20Pawn%20Object%20Reference) (InteractionBeingAttempted :type EInteractionBeingAttempted%20Enum :default "FirstInteraction"))
  (bind _depositamount (Variables|Design|QuestSetup|InteractionSettings|GetDepositAmount))
  (if (Variables|Default|GetbDepositToInteract)
    ;[1 1 1] Fixing Loc Issue with dynamic strings
    (switch Utilities|FlowControl|Switch|SwitchonE_Athena_GenericQuestDamageTypes (Variables|Design|QuestSetup|InteractionSettings|GetInteractionType)
      (:NewEnumerator13
        (return (Utilities|Text|ToText(String) (Utilities|String|Append (Utilities|String|ToString(Text) (Variables|Design|QuestSetup|InteractionSettings|GetInteraction%20Verb)) " " (Utilities|String|ToString(Text) (Utilities|Text|FormatText "NSLOCTEXT(\"[F0DB40674147727E909EBBB40FEF4ECB]\", \"2C544E0F4993497E64C8538B9419E4F1\", \"{amount} bars\")" _depositamount))))))
      (:NewEnumerator14
        (return (Utilities|Text|ToText(String) (Utilities|String|Append (Utilities|String|ToString(Text) (Variables|Design|QuestSetup|InteractionSettings|GetInteraction%20Verb)) " " (Utilities|String|ToString(Text) (Utilities|Text|FormatText "NSLOCTEXT(\"[F0DB40674147727E909EBBB40FEF4ECB]\", \"7A74F6E8459F8C30C1FC9F89C86257BC\", \"{amount} wood\")" _depositamount))))))
      (:NewEnumerator15
        (return (Utilities|Text|ToText(String) (Utilities|String|Append (Utilities|String|ToString(Text) (Variables|Design|QuestSetup|InteractionSettings|GetInteraction%20Verb)) " " (Utilities|String|ToString(Text) (Utilities|Text|FormatText "NSLOCTEXT(\"[F0DB40674147727E909EBBB40FEF4ECB]\", \"6318F288429F7BC2C37FB4B81B9D9A1A\", \"{amount} stone\")" _depositamount))))))
      (:NewEnumerator16
        (return (Utilities|Text|ToText(String) (Utilities|String|Append (Utilities|String|ToString(Text) (Variables|Design|QuestSetup|InteractionSettings|GetInteraction%20Verb)) " " (Utilities|String|ToString(Text) (Utilities|Text|FormatText "NSLOCTEXT(\"[F0DB40674147727E909EBBB40FEF4ECB]\", \"7F043F774742C8A6EAD94896360B35E5\", \"{amount} metal\")" _depositamount)))))))
    (else
      (return (Variables|Design|QuestSetup|InteractionSettings|GetInteraction%20Verb)))))

(fn BlueprintGetInteractionTime
  (params (InteractingPawn :type Fort%20Pawn%20Object%20Reference) (InteractionBeingAttempted :type EInteractionBeingAttempted%20Enum :default "FirstInteraction"))
  (Utilities|IsValid (Variables|Components|GetNonPlayerConversationComponent)
    (:Is%20Valid
      (return 0.0 true))
    (:Is%20Not%20Valid
      (return 0.0 false))))

(fn BlueprintOnInteract
  (params (InteractingPawn :type Fort%20Pawn%20Object%20Reference) (InteractionBeingAttempted :type EInteractionBeingAttempted%20Enum :default "FirstInteraction"))
  (bind _interactingpawn (Variables|GetInteractingPawn))
  (bind _returnvalue (Fortnite|GetFortPlayerControllerfromActor _interactingpawn))
  (bind _returnvalue_2 (Fortnite|GetFortPlayerControllerfromActor InteractingPawn))
  (bind _resourceitemdefinition (Variables|Default|GetResourceItemDefinition))
  (bind _bconversationtocomplete (Variables|Design|GetbConversationToComplete))
  (Utilities|IsValid (Variables|Components|GetNonPlayerConversationComponent)
    (:Is%20Valid
      (Utilities|IsValid _returnvalue_2
        (:Is%20Valid
          (if (Conversation|CanStartConversationWith (Variables|Components|GetNonPlayerConversationComponent) _returnvalue_2)
            (Conversation|StartConversation (Variables|Components|GetNonPlayerConversationComponent) (Utilities|Array|Get(acopy) (GameplayTags|BreakGameplayTagContainer (Variables|Design|QuestSetup|InteractionSettings|GetDamage%20or%20Interact%20Tag))) _returnvalue_2 self :!class /Script/FortniteGame.FortNonPlayerConversationParticipantComponent)
            (else
              _))
          (Development|PrintString "OnServerInteract")
          (Development|PrintString (Utilities|String|Append "Conversation to Complete:
" (Utilities|String|ToString(Boolean) _bconversationtocomplete)))
          (if _bconversationtocomplete
            _
            (elif (Variables|Design|QuestSetup|GetbTestVisualsNoQuestRequired)
              (CallFunction|ChangeStateforPlayer _returnvalue)
              (return)
              (elif (Variables|Default|GetbDepositToInteract)
                ;[1 1 1] Deposit to Interact logic
                (if (>= (Inventory|GetItemQuantityOnPlayer _returnvalue _resourceitemdefinition) (Variables|Design|QuestSetup|InteractionSettings|GetDepositAmount))
                  (bind _returnvalue_3 (InventoryManagement|FindExistingItemforDefinition _returnvalue _resourceitemdefinition))
                  (InventoryManagement|RemoveItem _returnvalue (Item|GetItemGuid _returnvalue_3) (Variables|Design|QuestSetup|InteractionSettings|GetDepositAmount))
                  (return))
                (elif (Variables|Design|QuestSetup|GetbSkipQuestObjectiveCompletion)
                  ;[1 1 1] Interact logic triggers state change independent of quest completion
                  (CallFunction|ForceInteractStateChange _returnvalue))))))))
    (:Is%20Not%20Valid
      _)))

(fn BlueprintOnLocalInteract
  (params (InteractingPawn :type Fort%20Player%20Pawn%20Object%20Reference) (InteractionBeingAttempted :type EInteractionBeingAttempted%20Enum :default "FirstInteraction"))
  (bind _returnvalue (Fortnite|GetFortPlayerControllerfromActor InteractingPawn))
  (Development|PrintString "BlueprintOnLocalInteract")
  (Variables|Default|SetbBlockLocalInteractOverride true)
  (if (Variables|Design|GetbConversationToComplete)
    _
    (elif (Variables|Design|QuestSetup|GetbTestVisualsNoQuestRequired)
      (CallFunction|ChangeStateforPlayer _returnvalue)
      (return false)
      (elif (Variables|Design|QuestSetup|GetbSkipQuestObjectiveCompletion)
        ;[1 1 1] Interact logic triggers state change independent of quest completion
        (CallFunction|ForceInteractStateChange _returnvalue)))))

(fn CheckTagsAndComplete
  (params (Actor :type Actor%20Object%20Reference))
  (bind _returnvalue (Fortnite|GetFortPlayerControllerfromActor Actor))
  (bind _returnvalue_2 (CallFunction|PlayerTagsCheck _returnvalue (Variables|Design|QuestSetup|GameplayTags,EffectsandCosmetics|GetRequired%20Player%20Tags) (Variables|Design|QuestSetup|GameplayTags,EffectsandCosmetics|GetProhibited%20Player%20Tags)))
  (if _returnvalue_2
    (if (Variables|Design|QuestSetup|GetbTestVisualsNoQuestRequired)
      (if (not (Variables|Debug|GetbOnVisualTestingCooldown))
        (CallFunction|LocalVisualsTest _returnvalue)
        (CallFunction|ChangeStateforPlayer _returnvalue))
      (elif (Quests|IsQuestObjectiveActive (Quests|GetQuestManager _returnvalue "Invalid") (Variables|/Script/FortniteGame.ConsolidatedQuestComponent|GetObjectiveBackendName (Variables|BuildingGameplayActorPropQuest|GetConsolidatedQuestComponent)))
        (if (Variables|Design|QuestSetup|GetbSkipQuestObjectiveCompletion)
          (CallFunction|ForceInteractStateChange _returnvalue))))))

(fn ConversationMessage
  (params (MessageNumber :type Integer)))

(fn ConversationSuccess
  (params (PlayerController :type Fort%20Player%20Controller%20Object%20Reference))
  (bind _returnvalue (Pawn|GetControlledPawn PlayerController))
  (Utilities|IsValid PlayerController
    (:Is%20Valid
      (Utilities|IsValid _returnvalue
        (:Is%20Valid
          (CallFunction|ConversationtoChangeState _returnvalue))))))

(fn CosmeticOverridestoPlayer
  (params (FPC :type Fort%20Player%20Controller%20Object%20Reference) (Remove :type Boolean))
  (bind _returnvalue (Game|Player|GetPlayerPawn FPC))
  (bind _returnvalue_2 (GUID|GenerateGuidfromObject (Class|ConsolidatedQuestComponent|GetQuestItemDefinition (Variables|BuildingGameplayActorPropQuest|GetConsolidatedQuestComponent))))
  (if Remove
    (Class|FortPlayerPawn|SwapPawnCosmeticItemandVariant _returnvalue _returnvalue_2 (Variables|Design|QuestSetup|GameplayTags,EffectsandCosmetics|GetPlayer%20Cosmetic%20Overrides) false "Inactive" false)
    (Development|PrintString "REMOVED")
    (else
      (Class|FortPlayerPawn|SwapPawnCosmeticItemandVariant _returnvalue _returnvalue_2 (Variables|Design|QuestSetup|GameplayTags,EffectsandCosmetics|GetPlayer%20Cosmetic%20Overrides) false "Active" false))))

(fn ManagePlayerGEs
  (params (FPC :type Fort%20Player%20Controller%20Object%20Reference))
  (bind _returnvalue (Game|Player|GetPlayerPawn FPC))
  (bind _interaction_ges (Variables|Design|QuestSetup|GameplayTags,EffectsandCosmetics|GetInteraction%20GEs))
  (CallFunction|QuestLog "Beginning Manage Player GEs")
  (if (Variables|Design|QuestSetup|GameplayTags,EffectsandCosmetics|GetbShouldApplyGEs)
    (CallFunction|QuestLog "Apply GEs")
    (CallFunction|QuestLog (Utilities|String|ToString(Integer) (Utilities|Array|Length _interaction_ges)))
    (for _array_element _interaction_ges
      (CallFunction|QuestLog (Utilities|GetClassDisplayName _array_element))
      (Environment|ApplyGlobalEnvironmentGameplayEffecttoActor _returnvalue _array_element))
    (else
      _))
  (if (Variables|Design|QuestSetup|GameplayTags,EffectsandCosmetics|GetbShouldRemoveGEs)
    (CallFunction|QuestLog "Remove GEs")
    (for _array_element_2 (Variables|Design|QuestSetup|GameplayTags,EffectsandCosmetics|GetRemove%20GEs)
      (CallFunction|QuestLog (Utilities|GetClassDisplayName _array_element_2))
      (GameplayEffects|RemoveActiveGameplayEffectbySourceEffect (Ability|GetAbilitySystemComponent _returnvalue) _array_element_2 0 -1))
    (else
      _))
  (return))

(fn PostInteractionVisuals
  (params (FPC :type Fort%20Player%20Controller%20Object%20Reference))
  (bind _postinteraction_mis_override (Variables|Design|PostInteraction|GetPostInteraction_MIS_Override))
  (bind _sm_quest_postinteraction (Variables|Components|GetSM_Quest_PostInteraction))
  (bind _sm_quest_preinteraction (Variables|Components|GetSM_Quest_PreInteraction))
  (bind _storedoriginalpreinteractionmaterial (Variables|Default|GetStoredOriginalPreInteractionMaterial))
  (bind _sm_quest_preinteraction_2 (Variables|Components|GetSM_Quest_PreInteraction))
  (Utilities|IsValid (Variables|Design|PostInteraction|GetSM_PostInteraction)
    (:Is%20Valid
      (if (PlayerController|IsinSplitscreen FPC)
        (PlayerController|SetCompHiddenControllerOnly FPC _sm_quest_preinteraction_2 true false)
        (PlayerController|SetCompHiddenControllerOnly FPC _sm_quest_postinteraction false true)
        (else
          (Development|SetHiddeninGame _sm_quest_preinteraction_2 true false)
          (Development|SetHiddeninGame _sm_quest_postinteraction false true)))
      (Transformation|SetRelativeTransform _sm_quest_postinteraction (Variables|Design|PostInteraction|GetPostInteraction%20Mesh%20Transform) false false))
    (:Is%20Not%20Valid
      _))
  (Utilities|IsValid (Variables|Design|PostInteraction|GetPostInteraction_MIS_Override)
    (:Is%20Valid
      (Utilities|IsValid (Variables|Design|PostInteraction|GetSM_PostInteraction)
        (:Is%20Valid
          (Rendering|Material|SetMaterial (Variables|Components|GetSM_Quest_PostInteraction) 0 _postinteraction_mis_override))
        (:Is%20Not%20Valid
          (Rendering|Material|SetMaterial _sm_quest_preinteraction 0 _postinteraction_mis_override))))
    (:Is%20Not%20Valid
      (Utilities|IsValid _storedoriginalpreinteractionmaterial
        (:Is%20Valid
          (Rendering|Material|SetMaterial _sm_quest_preinteraction 0 _storedoriginalpreinteractionmaterial))
        (:Is%20Not%20Valid
          _))))
  (Development|PrintString "End of Material" true true "(R=0.263882,G=1.000000,B=0.000000,A=1.000000)"))

(fn ShouldDie
  (params (Damage :type Float%20%28single-precision%29) (EventInstigator :type Controller%20Object%20Reference) (DamageCauser :type Actor%20Object%20Reference) (bIsTest :type Boolean) (EffectContext :type Gameplay%20Effect%20Context%20Handle%20Structure%20%28by%20ref%29))
  (return false))

(fn SpawnCosmeticActor
  (Networking|SwitchHasAuthority
    (:Remote
      (if (Utilities|IsValidClass (Variables|Design|CosmeticEvent|GetClientCosmeticActorClass))
        (Utilities|IsValid (Variables|Default|GetClientCosmeticActorRef)
          (:Is%20Valid
            (if (Variables|Design|QuestSetup|GetbTestVisualsNoQuestRequired)
              _))
          (:Is%20Not%20Valid
            _)) ; this prevents more than 1 per player per actor
        (bind _returnvalue (Game|SpawnActorfromClass (Variables|Design|CosmeticEvent|GetClientCosmeticActorClass) (Math|Transform|MakeTransform (Transformation|GetActorLocation self) (Transformation|GetActorRotation self)) "AlwaysSpawn" "SelectDefaultAtRuntime" 0 0))
        (Variables|Default|SetClientCosmeticActorRef _returnvalue)
        (Development|PrintString "Spawned!\\")))))

(fn UpdateOverlappingActors
  (bind _sm_hiddeninteractioncollision (Variables|Components|GetSM_HiddenInteractionCollision))
  (bind _overlappingactors (Collision|GetOverlappingActors _sm_hiddeninteractioncollision "/Script/FortniteGame.FortPlayerPawn" :!class /Script/Engine.PrimitiveComponent))
  (bind _collision_discoverradius (Variables|Components|GetCollision_DiscoverRadius))
  (bind _overlappingactors_2 (Collision|GetOverlappingActors _collision_discoverradius "/Script/FortniteGame.FortPlayerPawn" :!class /Script/Engine.PrimitiveComponent))
  (if (Variables|Design|QuestSetup|GetbRecheckingTriggersOnSuccess)
    _
    (elif (Variables|Design|QuestSetup|GetbReHideIfConditionsChange)
      _
      (else
        (return))))
  (if (Variables|Design|GetbUseDiscoverCollisionToUnhide)
    (if (Utilities|IsArrayNonEmpty _overlappingactors_2)
      (for _array_element _overlappingactors_2
        (CallFunction|UpdateOverlappingActors(CollisionDiscoverRadius) _array_element)))
    (else
      _))
  (if (Utilities|IsArrayNonEmpty _overlappingactors)
    (for _array_element_2 _overlappingactors
      (CallFunction|UpdateOverlappingActors(SMHiddenInteractionCollision) _array_element_2)))
  (return))

(fn ConstructionScript
  (bind (_origin _boxextent) (Collision|GetComponentBounds (Variables|Components|GetSM_HiddenInteractionCollision)))
  (bind (_origin_2 _boxextent_2) (Collision|GetComponentBounds (Variables|Components|GetSM_Quest_PreInteraction)))
  (bind _ambientsound (Variables|Design|PreInteraction|AmbientFX|GetAmbientSound))
  (bind _sm_hiddeninteractioncollision (Variables|Components|GetSM_HiddenInteractionCollision))
  (bind (_relativelocation_x _relativelocation_y) (Math|Vector|BreakVector (Variables|/Script/Engine.SceneComponent|GetRelativeLocation_X (Variables|Components|GetText_BackendNameHelper))))
  (bind _sm_hiddeninteractioncollision_2 (Variables|Components|GetSM_HiddenInteractionCollision))
  (bind _questiconcomponent (Variables|BuildingGameplayActorPropQuest|GetQuestIconComponent))
  (bind _sm_hiddeninteractioncollision_3 (Variables|Components|GetSM_HiddenInteractionCollision))
  (bind _sm_hiddeninteractioncollision_4 (Variables|Components|GetSM_HiddenInteractionCollision))
  (bind _p_postinteractionambient (Variables|Components|GetP_PostInteractionAmbient))
  (bind _p_ambient (Variables|Components|GetP_Ambient))
  (bind _preinteraction_mis_override (Variables|Design|PreInteraction|GetPreInteraction_MIS_Override))
  (bind _postinteractionambientsound (Variables|Design|PostInteraction|AmbientFX|GetPostInteractionAmbientSound))
  (bind _cosmeticsound (Variables|Design|CosmeticEvent|GetCosmeticSound))
  (bind _sm_hiddeninteractioncollision_5 (Variables|Components|GetSM_HiddenInteractionCollision))
  (bind _unhide_particle (Variables|Design|PreInteraction|UnhideFX|GetUnhide%20Particle))
  (bind _unhidesound (Variables|Design|PreInteraction|UnhideFX|GetUnhideSound))
  (bind _postinteraction_mis_override (Variables|Design|PostInteraction|GetPostInteraction_MIS_Override))
  (bind _sm_hiddeninteractioncollision_6 (Variables|Components|GetSM_HiddenInteractionCollision))
  (bind _ambientniagarasystem (Variables|Design|PreInteraction|AmbientFX|GetAmbientNiagaraSystem))
  (bind _unhideniagarasystem (Variables|Design|PreInteraction|UnhideFX|GetUnhideNiagaraSystem))
  (bind _postinteractionambientniagarasystem (Variables|Design|PostInteraction|AmbientFX|GetPostInteractionAmbientNiagaraSystem))
  (bind _ambientparticle (Variables|Design|PreInteraction|AmbientFX|GetAmbientParticle))
  (bind _postinteractionambientparticle (Variables|Design|PostInteraction|AmbientFX|GetPostInteractionAmbientParticle))
  (bind _cosmeticparticle (Variables|Design|CosmeticEvent|GetCosmeticParticle))
  (bind _cosmeticniagarasystem (Variables|Design|CosmeticEvent|GetCosmeticNiagaraSystem))
  (bind _sm_hiddeninteractioncollision_7 (Variables|Components|GetSM_HiddenInteractionCollision))
  (Utilities|FlowControl|Sequence
    (:then_0
      ;[1 1 1] Master Quest Type Init
      (Variables|Design|QuestSetup|GameplayTags,EffectsandCosmetics|SetbShouldApplyGEs (Utilities|IsArrayNonEmpty (Variables|Design|QuestSetup|GameplayTags,EffectsandCosmetics|GetInteraction%20GEs)))
      (Variables|Design|QuestSetup|GameplayTags,EffectsandCosmetics|SetbShouldRemoveGEs (Utilities|IsArrayNonEmpty (Variables|Design|QuestSetup|GameplayTags,EffectsandCosmetics|GetRemove%20GEs)))
      (Variables|Design|QuestSetup|GameplayTags,EffectsandCosmetics|SetbShouldApplyCosmeticOverrides (Utilities|IsArrayNonEmpty (Variables|Design|QuestSetup|GameplayTags,EffectsandCosmetics|GetPlayer%20Cosmetic%20Overrides)))
      (switch Utilities|FlowControl|Switch|SwitchonE_Athena_GenericQuestActorTypes (Variables|Design|QuestSetup|GetbQuestActorType)
        (:NewEnumerator0
          (Variables|Design|SetbCompleteFromDiscovery false)
          (Variables|Design|SetbUseDiscoverCollisionToUnhide false))
        (:NewEnumerator1
          (Variables|Design|SetbCompleteFromDiscovery false)
          (Variables|Design|SetbUseDiscoverCollisionToUnhide true))
        (:NewEnumerator2
          (Variables|Design|SetbCompleteFromDiscovery true)
          (Variables|Design|SetbUseDiscoverCollisionToUnhide true))))
    (:then_1
      ;[1 1 1] Pre and Post Interaction Meshes Init
      (Components|StaticMesh|SetStaticMesh (Variables|Components|GetSM_Quest_PreInteraction) (Variables|Design|PreInteraction|GetSM_PreInteraction))
      (Components|StaticMesh|SetStaticMesh (Variables|Components|GetSM_Quest_PostInteraction) (Variables|Design|PostInteraction|GetSM_PostInteraction))
      (Transformation|SetRelativeTransform (Variables|Components|GetSM_Quest_PostInteraction) (Variables|Design|PostInteraction|GetPostInteraction%20Mesh%20Transform) false false)
      (Transformation|SetRelativeTransform (Variables|Components|GetSM_Quest_PreInteraction) (Variables|Design|PreInteraction|GetPreInteraction%20Mesh%20Transform) false false)
      ;;[1 1 1] Hide Or Not Post Interaction Mesh
      (if (Variables|Design|PostInteraction|GetbPreviewPostInteraction)
        (Utilities|IsValid (Variables|Design|PostInteraction|GetSM_PostInteraction)
          (:Is%20Valid
            (Transformation|SetWorldScale3D (Variables|Components|GetSM_Quest_PreInteraction) "0, 0, 0"))
          (:Is%20Not%20Valid
            _))
        (else
          (Transformation|SetWorldScale3D (Variables|Components|GetSM_Quest_PostInteraction) "0, 0, 0")))
      ;;[1 1 1] Change based on if we are using "discovery" or not
      (if (Variables|Design|GetbUseDiscoverCollisionToUnhide)
        (Transformation|SetWorldScale3D (Variables|Components|GetSM_Quest_PreInteraction) "0, 0, 0")
        (else
          (Components|DestroyComponent (Variables|Components|GetCollision_DiscoverRadius)))))
    (:then_2
      ;[1 1 1] Particle Init
      (Transformation|SetRelativeTransform (Variables|Components|GetP_Ambient) (Variables|Design|PreInteraction|AmbientFX|GetAmbient%20Particle%20Transform) false true)
      (Utilities|IsValid _ambientparticle
        (:Is%20Valid
          (Effects|Components|ParticleSystem|SetTemplate (Variables|Components|GetP_Ambient) _ambientparticle))
        (:Is%20Not%20Valid
          _))
      (Utilities|IsValid _ambientniagarasystem
        (:Is%20Valid
          (Niagara|SetNiagaraSystemAsset (Variables|Components|GetN_Ambient) _ambientniagarasystem true))
        (:Is%20Not%20Valid
          _))
      (Utilities|IsValid _postinteractionambientniagarasystem
        (:Is%20Valid
          (Niagara|SetNiagaraSystemAsset (Variables|Components|GetN_PostInteractionAmbient) _postinteractionambientniagarasystem true))
        (:Is%20Not%20Valid
          _))
      (Utilities|IsValid _postinteractionambientparticle
        (:Is%20Valid
          (Effects|Components|ParticleSystem|SetTemplate (Variables|Components|GetP_PostInteractionAmbient) _postinteractionambientparticle))
        (:Is%20Not%20Valid
          _))
      (Transformation|SetRelativeTransform (Variables|Components|GetP_PostInteractionAmbient) (Variables|Design|PostInteraction|AmbientFX|GetPostInteractionAmbientParticleTransform) false false)
      (if (Variables|Design|QuestSetup|GetbTestVisualsNoQuestRequired)
        (if (Variables|Design|PostInteraction|GetbPreviewPostInteraction)
          (Components|Activation|SetAutoActivate _p_ambient false)
          (Components|Activation|SetAutoActivate _p_postinteractionambient true)
          (else
            (Components|Activation|SetAutoActivate _p_ambient true)
            (Components|Activation|SetAutoActivate _p_postinteractionambient false)))
        (else
          _))
      (Utilities|IsValid _cosmeticparticle
        (:Is%20Valid
          (Effects|Components|ParticleSystem|SetTemplate (Variables|Components|GetP_Cosmetic) _cosmeticparticle))
        (:Is%20Not%20Valid
          _))
      (Utilities|IsValid _cosmeticniagarasystem
        (:Is%20Valid
          (Niagara|SetNiagaraSystemAsset (Variables|Components|GetN_Cosmetic) _cosmeticniagarasystem true))
        (:Is%20Not%20Valid
          _))
      (Transformation|SetRelativeTransform (Variables|Components|GetP_Cosmetic) (Variables|Design|CosmeticEvent|GetCosmeticParticleTransform) false false)
      (Utilities|IsValid _unhide_particle
        (:Is%20Valid
          (Effects|Components|ParticleSystem|SetTemplate (Variables|Components|GetP_UnHide) _unhide_particle))
        (:Is%20Not%20Valid
          _))
      (Utilities|IsValid _unhideniagarasystem
        (:Is%20Valid
          (Niagara|SetNiagaraSystemAsset (Variables|Components|GetN_UnHide) _unhideniagarasystem true))
        (:Is%20Not%20Valid
          _))
      (Transformation|SetRelativeTransform (Variables|Components|GetN_UnHide) (Variables|Design|PreInteraction|UnhideFX|GetUnhideParticleTransform) false false))
    (:then_3
      ;[1 1 1] Sound Init
      (Utilities|IsValid _ambientsound
        (:Is%20Valid
          (Audio|Components|Audio|SetSound (Variables|Components|GetAmbientAudio) _ambientsound))
        (:Is%20Not%20Valid
          _))
      (Utilities|IsValid _postinteractionambientsound
        (:Is%20Valid
          (Audio|Components|Audio|SetSound (Variables|Components|GetPostInteractionAmbientAudio) _postinteractionambientsound))
        (:Is%20Not%20Valid
          _))
      (Utilities|IsValid _cosmeticsound
        (:Is%20Valid
          (Audio|Components|Audio|SetSound (Variables|Components|GetCosmeticAudio) _cosmeticsound))
        (:Is%20Not%20Valid
          _))
      (Utilities|IsValid _unhidesound
        (:Is%20Valid
          (Audio|Components|Audio|SetSound (Variables|Components|GetUnHideAudio) _unhidesound))))
    (:then_4
      ;[1 1 1] Radii Setup
      (Components|Sphere|SetSphereRadius (Variables|Components|GetCollision_AmbientAudioAndFXActivation) (Variables|Design|QuestSetup|GetAudioAndVFXHideRadius) false)
      (Components|Sphere|SetSphereRadius (Variables|Components|GetCollision_DiscoverRadius) (Variables|Design|QuestSetup|GetDiscoverRadius) false))
    (:then_5
      ;[1 1 1] MiniMap Setup
      (if (or (Variables|Design|QuestSetup|MinimapIcon|GetbShowMinimapIcon) (Variables|Design|QuestSetup|MinimapIcon|GetbShowBigMapIcon))
        (MiniMapData|SetIcon (Variables|BuildingGameplayActorPropQuest|GetQuestIconComponent) (Variables|Design|QuestSetup|MinimapIcon|GetMini%20Map%20Icon))
        (switch Utilities|FlowControl|Switch|SwitchonE_Athena_GenericQuestActorMinimapRanges (Variables|Design|QuestSetup|MinimapIcon|GetMapRevealDistance)
          (:NewEnumerator0
            (Variables|Design|QuestSetup|MinimapIcon|SetView%20Distance 1024.0))
          (:NewEnumerator1
            (Variables|Design|QuestSetup|MinimapIcon|SetView%20Distance 2.5e+03))
          (:NewEnumerator2
            (Variables|Design|QuestSetup|MinimapIcon|SetView%20Distance 5e+03))
          (:NewEnumerator3
            (Variables|Design|QuestSetup|MinimapIcon|SetView%20Distance 7.5e+03))
          (:NewEnumerator4
            (Variables|Design|QuestSetup|MinimapIcon|SetView%20Distance 1.2e+04))
          (:NewEnumerator5
            (Variables|Design|QuestSetup|MinimapIcon|SetView%20Distance (Variables|Design|QuestSetup|MinimapIcon|GetCustomViewDistance))))
        (MiniMapData|SetViewDistanceRadius (Variables|BuildingGameplayActorPropQuest|GetQuestIconComponent) (Variables|Design|QuestSetup|MinimapIcon|GetView%20Distance))
        (MiniMapData|SetIconVisibilityonMiniMap _questiconcomponent (Variables|Design|QuestSetup|MinimapIcon|GetbShowMinimapIcon))
        (MiniMapData|SetIconVisibilityonMap _questiconcomponent (Variables|Design|QuestSetup|MinimapIcon|GetbShowBigMapIcon))
        (else
          (Components|DestroyComponent (Variables|BuildingGameplayActorPropQuest|GetQuestIconComponent)))))
    (:then_6
      ;[1 1 1] Material Swaps
      (Utilities|IsValid _preinteraction_mis_override
        (:Is%20Valid
          (Variables|Default|SetStoredOriginalPreInteractionMaterial (Utilities|Array|Get(acopy) (Rendering|Material|GetMaterials (Variables|Components|GetSM_Quest_PreInteraction))))
          (if (not (Variables|Design|PostInteraction|GetbPreviewPostInteraction))
            ;;[1 1 1] If PreInteraction has swapped material
            (Rendering|Material|SetMaterial (Variables|Components|GetSM_Quest_PreInteraction) 0 _preinteraction_mis_override)))
        (:Is%20Not%20Valid
          (if (Variables|Design|PostInteraction|GetbPreviewPostInteraction)
            _)))
      ;;[1 1 1] If Previewing the Post Interaction State
      (Utilities|IsValid _postinteraction_mis_override
        (:Is%20Valid
          (Utilities|IsValid (Variables|Design|PostInteraction|GetSM_PostInteraction)
            (:Is%20Valid
              (Rendering|Material|SetMaterial (Variables|Components|GetSM_Quest_PostInteraction) 0 _postinteraction_mis_override))
            (:Is%20Not%20Valid
              (Rendering|Material|SetMaterial (Variables|Components|GetSM_Quest_PreInteraction) 0 _postinteraction_mis_override))))
        (:Is%20Not%20Valid
          (Rendering|Material|SetMaterial (Variables|Components|GetSM_Quest_PreInteraction) 0 (Variables|Default|GetStoredOriginalPreInteractionMaterial)))))
    (:then_7
      ;[1 1 1] Gross Damage to Interact Switcher
      ;;[1 1 1] Init
      (Collision|SetCollisionProfileName _sm_hiddeninteractioncollision_3 "FortTriggerOnlyInteractions" true :!class /Script/Engine.PrimitiveComponent)
      (Transformation|SetWorldScale3D _sm_hiddeninteractioncollision_3 (Variables|Design|QuestSetup|InteractionSettings|GetInteraction%20Collision%20Scale))
      (Variables|/Script/Engine.PrimitiveComponent|SetbGenerateOverlapEvents false _sm_hiddeninteractioncollision_3)
      (Variables|Default|SetbOverlapToInteract false)
      (Variables|Default|SetbDamageToInteract false)
      (Variables|Default|SetbDepositToInteract false)
      (Variables|Default|SetbListenToTagToInteract false)
      (Variables|Design|SetbConversationToComplete false)
      (CallFunction|QuestLog "In construction script")
      (switch Utilities|FlowControl|Switch|SwitchonE_Athena_GenericQuestDamageTypes (Variables|Design|QuestSetup|InteractionSettings|GetInteractionType)
        (:NewEnumerator0
          ;;[1 1 1] Tag Templates
          (Variables|Design|QuestSetup|InteractionSettings|SetDamage%20or%20Interact%20Tag "(GameplayTags=)"))
        (:NewEnumerator5
          (Variables|Design|QuestSetup|InteractionSettings|SetDamage%20or%20Interact%20Tag "(GameplayTags=((TagName=\"Weapon\")))"))
        (:NewEnumerator12
          (Variables|Design|QuestSetup|InteractionSettings|SetDamage%20or%20Interact%20Tag "(GameplayTags=((TagName=\"Weapon.Melee.Impact.Pickaxe\")))"))
        (:NewEnumerator6
          (Variables|Design|QuestSetup|InteractionSettings|SetDamage%20or%20Interact%20Tag "(GameplayTags=((TagName=\"Weapon.Melee\")))"))
        (:NewEnumerator7
          (Variables|Design|QuestSetup|InteractionSettings|SetDamage%20or%20Interact%20Tag "(GameplayTags=((TagName=\"Weapon.Ranged\")))"))
        (:NewEnumerator8
          (Variables|Design|QuestSetup|InteractionSettings|SetDamage%20or%20Interact%20Tag "(GameplayTags=((TagName=\"Weapon.Ranged.Grenade.Standard\"),(TagName=\"Weapon.Ranged.Heavy.C4\"),(TagName=\"Weapon.Ranged.Heavy.Grenade\"),(TagName=\"Weapon.Ranged.Heavy.Grenade_Launcher\"),(TagName=\"Weapon.Ranged.Heavy.Mortar\"),(TagName=\"Weapon.Ranged.Heavy.ProximityMine\"),(TagName=\"Weapon.Ranged.Heavy.Rocket_Launcher\")))")
          ;;[1 1 1] Collision Switching
          (Collision|SetCollisionProfileName _sm_hiddeninteractioncollision_4 "FortTriggerOnlyWeapons" true :!class /Script/Engine.PrimitiveComponent)
          ;;[1 1 1] Logic Mode Switching
          (Variables|/Script/Engine.PrimitiveComponent|SetbGenerateOverlapEvents true _sm_hiddeninteractioncollision_4)
          (Variables|Default|SetbDamageToInteract true))
        (:NewEnumerator9
          (Variables|Design|QuestSetup|InteractionSettings|SetDamage%20or%20Interact%20Tag "(GameplayTags=)")
          (Collision|SetCollisionProfileName _sm_hiddeninteractioncollision "FortTriggerOnlyPawns" true :!class /Script/Engine.PrimitiveComponent)
          (Collision|SetCollisionResponsetoAllChannels _sm_hiddeninteractioncollision "ECR_Ignore")
          (Collision|SetCollisionResponsetoChannel _sm_hiddeninteractioncollision "ECC_Vehicle" "ECR_Overlap")
          (Variables|/Script/Engine.PrimitiveComponent|SetbGenerateOverlapEvents true _sm_hiddeninteractioncollision)
          (Variables|Default|SetbOverlapToInteract true))
        (:NewEnumerator10
          (Variables|Design|QuestSetup|InteractionSettings|SetDamage%20or%20Interact%20Tag "(GameplayTags=)"))
        (:NewEnumerator11
          _)
        (:NewEnumerator13
          (Variables|Default|SetResourceItemDefinition "/Game/Items/ResourcePickups/Athena_WadsItemData.Athena_WadsItemData")
          (Variables|Design|QuestSetup|InteractionSettings|SetDamage%20or%20Interact%20Tag "(GameplayTags=)")
          (Variables|Default|SetbDepositToInteract true))
        (:NewEnumerator14
          (Variables|Default|SetResourceItemDefinition "/Game/Items/ResourcePickups/WoodItemData.WoodItemData"))
        (:NewEnumerator15
          (Variables|Default|SetResourceItemDefinition "/Game/Items/ResourcePickups/StoneItemData.StoneItemData"))
        (:NewEnumerator16
          (Variables|Default|SetResourceItemDefinition "/Game/Items/ResourcePickups/MetalItemData.MetalItemData"))
        (:NewEnumerator17
          (Collision|SetCollisionProfileName _sm_hiddeninteractioncollision_2 "FortTriggerOnlyPawns" true :!class /Script/Engine.PrimitiveComponent)
          (Collision|SetCollisionResponsetoAllChannels _sm_hiddeninteractioncollision_2 "ECR_Ignore")
          (Collision|SetCollisionResponsetoChannel _sm_hiddeninteractioncollision_2 "ECC_Pawn" "ECR_Overlap")
          (Variables|/Script/Engine.PrimitiveComponent|SetbGenerateOverlapEvents true _sm_hiddeninteractioncollision_2)
          (Variables|Default|SetbOverlapToInteract true)
          (Variables|Default|SetbListenToTagToInteract true))
        (:NewEnumerator18
          (Collision|SetCollisionProfileName _sm_hiddeninteractioncollision_7 "FortProjectileBall" true :!class /Script/Engine.PrimitiveComponent)
          (Collision|SetCollisionResponsetoAllChannels _sm_hiddeninteractioncollision_7 "ECR_Ignore")
          (Collision|SetCollisionResponsetoChannel _sm_hiddeninteractioncollision_7 "ECC_GameTraceChannel8" "ECR_Overlap")
          (Variables|/Script/Engine.PrimitiveComponent|SetbGenerateOverlapEvents true _sm_hiddeninteractioncollision_7)
          (Variables|Default|SetbOverlapToInteract true))
        (:NewEnumerator19
          (Variables|Design|SetbConversationToComplete true))
        (:NewEnumerator20
          (Collision|SetCollisionProfileName _sm_hiddeninteractioncollision_6 "FortTriggerOnlyPawns" true :!class /Script/Engine.PrimitiveComponent)
          (Collision|SetCollisionResponsetoAllChannels _sm_hiddeninteractioncollision_6 "ECR_Ignore")
          (Collision|SetCollisionResponsetoChannel _sm_hiddeninteractioncollision_6 "ECC_Pawn" "ECR_Overlap")
          (Variables|/Script/Engine.PrimitiveComponent|SetbGenerateOverlapEvents true _sm_hiddeninteractioncollision_6)
          (Variables|Default|SetbOverlapToInteract true))
        (:NewEnumerator21
          (Collision|SetCollisionProfileName _sm_hiddeninteractioncollision_5 "FortTriggerOnlyPawns" true :!class /Script/Engine.PrimitiveComponent)
          (Collision|SetCollisionResponsetoAllChannels _sm_hiddeninteractioncollision_5 "ECR_Ignore")
          (Collision|SetCollisionResponsetoChannel _sm_hiddeninteractioncollision_5 "ECC_Pawn" "ECR_Overlap")
          (Collision|SetCollisionResponsetoChannel _sm_hiddeninteractioncollision_5 "ECC_Vehicle" "ECR_Overlap")
          (Variables|/Script/Engine.PrimitiveComponent|SetbGenerateOverlapEvents true _sm_hiddeninteractioncollision_5)
          (Variables|Default|SetbOverlapToInteract true)))
      (Collision|SetCollisionProfileName (Variables|Components|GetSM_HiddenInteractionCollision) "FortTriggerOnlyWeapons" true :!class /Script/Engine.PrimitiveComponent)
      (Variables|Default|SetbDamageToInteract true))
    (:then_8
      ;[1 1 1] Editor Only
      (Variables|Default|SetbErrorMessage false)
      (Variables|Default|SetDebugText "")
      (if (Variables|Design|QuestSetup|GetbTestVisualsNoQuestRequired)
        (Variables|Default|SetbErrorMessage true)
        (Variables|Default|SetDebugText (Utilities|String|Append (Variables|Default|GetDebugText) "
Visual Test Mode"))
        (else
          _))
      (if (Variables|Design|PostInteraction|GetbPreviewPostInteraction)
        (Variables|Default|SetbErrorMessage true)
        (Variables|Default|SetDebugText (Utilities|String|Append (Variables|Default|GetDebugText) "
Preview Post-Interaction"))
        (else
          _))
      (Utilities|IsValid (Class|ConsolidatedQuestComponent|GetQuestItemDefinition (Variables|BuildingGameplayActorPropQuest|GetConsolidatedQuestComponent))
        (:Is%20Valid
          _)
        (:Is%20Not%20Valid
          (Variables|Default|SetbErrorMessage true)
          (Variables|Default|SetDebugText (Utilities|String|Append (Variables|Default|GetDebugText) "
Need QuestItemDefinition"))))
      (if (not (>= (Utilities|String|Len (Utilities|String|ToString(Name) (Variables|/Script/FortniteGame.ConsolidatedQuestComponent|GetObjectiveBackendName (Variables|BuildingGameplayActorPropQuest|GetConsolidatedQuestComponent)))) 1))
        (Variables|Default|SetbErrorMessage true)
        (Variables|Default|SetDebugText (Utilities|String|Append (Variables|Default|GetDebugText) "
Need Objective Backend Name")))
      (if (and (Variables|Design|GetbUseDiscoverCollisionToUnhide) (> (Variables|Design|QuestSetup|GetAudioAndVFXHideRadius) (Variables|Design|QuestSetup|GetDiscoverRadius)))
        (Variables|Default|SetbErrorMessage true)
        (Variables|Default|SetDebugText (Utilities|String|Append (Variables|Default|GetDebugText) "
Discovery Radius Smaller than VFX radius"))
        (else
          _))
      (if (and (Variables|Default|GetbListenToTagToInteract) (not (Utilities|GameplayTagContainerNotEmpty (Variables|Design|QuestSetup|InteractionSettings|GetDamage%20or%20Interact%20Tag))))
        (Variables|Default|SetbErrorMessage true)
        (Variables|Default|SetDebugText (Utilities|String|Append (Variables|Default|GetDebugText) "
Need a Tag to listen to"))
        (else
          _))
      ;;[1 1 1] Positioning the Tooltip
      (Transformation|SetRelativeLocationAndRotation (Variables|Components|GetText_BackendNameHelper) (Math|Vector|MakeVector _relativelocation_x _relativelocation_y (* (Math|Float|Max(Float) (.z _boxextent_2) (.z _boxextent)) 2.0)) (Math|Rotator|InvertRotator (Transformation|GetActorRotation)) false false)
      (if (Variables|Default|GetbErrorMessage)
        (Rendering|Components|TextRender|SetTextRenderColor (Variables|Components|GetText_BackendNameHelper) (Variables|Default|GetErrorTextColor))
        (Rendering|Components|TextRender|SetText (Variables|Components|GetText_BackendNameHelper) (Utilities|Text|ToText(String) (Variables|Default|GetDebugText)))
        (else
          (Rendering|Components|TextRender|SetTextRenderColor (Variables|Components|GetText_BackendNameHelper) (Variables|Default|GetNormalTextColor))
          (Rendering|Components|TextRender|SetText (Variables|Components|GetText_BackendNameHelper) (Utilities|Text|ToText(String) (Utilities|String|Append "Obj: " (Utilities|String|ToString(Name) (Variables|/Script/FortniteGame.ConsolidatedQuestComponent|GetObjectiveBackendName (Variables|BuildingGameplayActorPropQuest|GetConsolidatedQuestComponent))) "
Tag: " (GameplayTags|GetDebugStringfromGameplayTagContainer (CallFunction|GetQuestTargetTags self)))))))
      (if (Variables|Design|QuestSetup|GetbTestVisualsNoQuestRequired)
        (Development|PrintString (Utilities|String|Append "Do Not Save with \"TEST VISUALS\" flag on!   " (Utilities|GetDisplayName self)) true true "(R=1.000000,G=0.000000,B=0.000000,A=1.000000)" 1.0)))
    (:then_10
      ;[1 1 1] Enable Conversation Component
      (if (not (Variables|Design|GetbConversationToComplete))
        (Components|DestroyComponent (Variables|Components|GetNonPlayerConversationComponent))))))
