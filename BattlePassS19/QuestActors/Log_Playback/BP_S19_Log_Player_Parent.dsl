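; Base quest prop for Season 19 log playback quest actors, built on the consolidated quest component inherited from BuildingGameplayActorPropQuest.
; When an actor enters its sphere, it checks whether the pre-interaction mesh uses the CB radio material and the actor has a valid player controller.
; If that player's Athena quest objective is not active, the prop destroys itself.
; BeginPlay, Tick and actor overlap events add no logic.
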(event EventBeginPlay
  (|CallParentFunction) ; This node is disabled and will not be called.\nDrag off pins to build functionality.
)

(event Collision|EventActorBeginOverlap
  (params (OtherActor :type Actor%20Object%20Reference)))

(event EventTick
  (params (DeltaSeconds :type Float%20%28single-precision%29)))

(event OnComponentBeginOverlap(Sphere)
  (params (OverlappedComponent :type Primitive%20Component%20Object%20Reference) (OtherActor :type Actor%20Object%20Reference) (OtherComp :type Primitive%20Component%20Object%20Reference) (OtherBodyIndex :type Integer) (bFromSweep :type Boolean) (SweepResult :type Hit%20Result%20Structure%20%28by%20ref%29))
  (bind _returnvalue (Fortnite|GetFortPlayerControllerfromActor OtherActor))
  (bind _self self)
  (if (== (Rendering|Material|GetMaterial (Variables|Components|GetSM_Quest_PreInteraction) 0) "/Game/Environments/Props/Industrial/Materials/MI_CBRadio_01.MI_CBRadio_01")
    (Utilities|IsValid _returnvalue
      (:Is%20Valid
        (if (Quests|IsQuestObjectiveActive (Quests|GetQuestManager _returnvalue "Athena") (Variables|/Script/FortniteGame.ConsolidatedQuestComponent|GetObjectiveBackendName (Variables|BuildingGameplayActorPropQuest|GetConsolidatedQuestComponent)))
          _
          (elif (Utilities|IsValid _self :!class /Script/Engine.KismetSystemLibrary)
            (Actor|DestroyActor _self)))))))

(fn ConstructionScript
  (|CallParentFunction))
